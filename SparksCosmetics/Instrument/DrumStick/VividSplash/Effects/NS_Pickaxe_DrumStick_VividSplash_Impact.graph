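# Niagara impact effect for the VividSplash drumstick pickaxe: a one-shot magenta, purple and blue splash.
# FN_PhysMatDebris throws mesh chunks whose launch and gravity follow User.PhysicalSurface, with collision disabled.
# Smoke, curl-noise sparks with glow sprites and an expanding wave join Light001, which packs flashes, streaks, mesh and light into one emitter via spawn-group masks.
# Front-end flags tone down brightness and offsets; time-of-day curves adjust some brightness.
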NiagaraSystem /SparksCosmetics/Instrument/DrumStick/VividSplash/Effects/NS_Pickaxe_DrumStick_VividSplash_Impact
dependencies {
 usedDataInterfaces [&NiagaraDataInterfaceCurve, &NiagaraDataInterfaceVectorCurve, &NiagaraDataInterfaceColorCurve]
 usedDynamicInputs []
 usedModules [
  &self:SystemSpawnScript.SystemScriptSource.SystemScriptGraph.NiagaraNodeAssignment_0.SetVariables_#1
  &SystemState
 ]
}
emitters {
 FN_PhysMatDebris {
  emitterUpdateScript [
   &EmitterState()
   &SpawnBurst_Instantaneous(
    Spawn Count=&RandomRangeInt(Minimum=5, Maximum=10)
    Use Spawn Probability=false
    UseLoopCountLimit=false
   )
  ]
  particleSpawnScript [
   /SetVariables(Particles.BiasedRandom=&FloatFromCurve(
     FloatCurve=DataInterface() {propertyValues {Curve {keys [
         <at 0.2 1 RCIM_Cubic>
         {arriveTangent=-3.3394, interpMode=RCIM_Cubic, leaveTangent=-3.3394, tangentMode=RCTM_User, time=0.5, value=0.5}
         <at 0.8 0 RCIM_Cubic>
        ]}}}
     CurveIndex=&RandomRangeFloat(Minimum=0, Maximum=1)
    )) {isSetParameters=true}
   &InitializeParticle(
    Color Mode=::NewEnumerator0
    Lifetime=&RandomRangeFloat(Minimum=0.5, Maximum=1.25)
    Mass=&Lerp_Float(A=0.6, B=1.2, Alpha=$Particles.BiasedRandom)
    Mass Mode=::NewEnumerator1
    Mesh Scale=&Multiply_VectorByFloat(
     Vector=&RandomRangeVector(Minimum=<0.9 0.4 0.5>, Maximum=<1 1 1>)
     Float=&Lerp_Float(A=0.7, B=1, Alpha=$Particles.BiasedRandom)
    )
    Mesh Scale Mode=::NewEnumerator4
    UsePositionOffset=false
    Write Material Random=true
   )
   &FN_PhysicalSurface(
    Dirt Mesh Count=2
    Glass Mesh Count=3
    Ice Mesh Count=2
    Metal Gravity=&RandomRangeFloat(Minimum=-350, Maximum=-1800)
    Metal Mesh Count=2
    Metal Velocity=&RandomRangeFloat(Minimum=80, Maximum=375)
    Metal Velocity Cone Angle=130
    Physical Surface=$User.PhysicalSurface
    Stone Gravity=&RandomRangeFloat(Minimum=-700, Maximum=-1250)
    Stone Mesh Count=2
    Stone Velocity=&RandomRangeFloat(Minimum=80, Maximum=550)
    Stone Velocity Cone Angle=120
    Wood Gravity=&RandomRangeFloat(Minimum=-450, Maximum=-900)
    Wood Mesh Count=2
    Wood Scale=1.5
    Wood Velocity=&RandomRangeFloat(Minimum=80, Maximum=600)
    Wood Velocity Cone Angle=130
   )
   &AddVelocityInCone(
    Cone Angle=$Output.FN_PhysicalSurface.SurfaceResponseConeAngle
    Module.Use Velocity Falloff On Cone Axis=false
    Velocity Strength=$Output.FN_PhysicalSurface.SurfaceResponseVelocity
   )
   &InitialMeshOrientation(Use Rotation Vector=false)
   &AddRotationalVelocity(Rotation Rate=&RandomVector(Vector Scale=&RandomRangeFloat(Minimum=1, Maximum=2)))
   &ApplyInitialForces()
   /SetVariables(Particles.GravityOverrideValue=$Output.FN_PhysicalSurface.GravityOverride) {isSetParameters=true}
  ]
  particleUpdateScript [
   &ParticleState()
   &GravityForce(Gravity=$Particles.GravityOverrideValue)
   &Collision(
    Advanced Aging Rate=1
    Age Colliding Particles=false
    CPU Collision Trace Channel=ECC_WorldStatic
    Collision Enabled=&Boolean_AndOperation(
     Bool1=&SetBoolByFloatComparison(A=$Particles.Initial.Lifetime, B=1)
     Bool2=&SetBoolByFloatComparison(A=$Particles.Age, B=0.12)
    )
    Control Roll On Collision=true
    Friction=1
    Friction During a Bounce=0.25
    Max CPU Trace Length=100
    Module.Control Rest State Via Surface Normal=false
    Module.EnableMaxCollisionCount=false
    Module.Randomize Collision Normal Vector=true
    Randomize Collision Normal=0.25
    ScaleRateOfRotation=false
   ) {enabled=false}
   &SolveForcesAndVelocity(Clamp Velocity=false, Limit Acceleration=false)
   &ScaleMeshSize(Scale Factor=&VectorFromFloat(Value=&FloatFromCurve(FloatCurve=DataInterface() {propertyValues {Curve {keys [<at 0.6 1 RCIM_Cubic>, {arriveTangent=-6.308, interpMode=RCIM_Cubic, leaveTangent=-6.308, tangentMode=RCTM_User, time=1, value=0}]}}})))
   &DynamicMaterialParameters(Module.Param0WriteEnabled=true, Module.Param1WriteEnabled=true, Module.Param2WriteEnabled=false, Module.Param3WriteEnabled=false)
  ]
  rendererClasses [&NiagaraSpriteRendererProperties, &NiagaraMeshRendererProperties]
  renderers [
   {rendererClass=&NiagaraSpriteRendererProperties}
   {rendererClass=&NiagaraMeshRendererProperties, rendererIndex=1}
   {rendererClass=&NiagaraMeshRendererProperties, rendererIndex=2}
   {rendererClass=&NiagaraMeshRendererProperties, rendererIndex=3}
   {rendererClass=&NiagaraMeshRendererProperties, rendererIndex=4}
   {rendererClass=&NiagaraMeshRendererProperties, rendererIndex=5}
   {rendererClass=&NiagaraMeshRendererProperties, rendererIndex=6}
   {rendererClass=&NiagaraMeshRendererProperties, rendererIndex=7}
   {rendererClass=&NiagaraMeshRendererProperties, rendererIndex=8}
  ]
 }
 Impact_Smoke {
  emitterUpdateScript [
   &EmitterState()
   &SpawnBurst_Instantaneous(
    Spawn Count=&RandomRangeInt(Minimum=2, Maximum=3)
    Use Spawn Probability=false
    UseLoopCountLimit=false
   )
  ]
  particleSpawnScript [
   &InitializeParticle(
    Color Maximum=<rgba 0.4957 0.4387 1 0.2>
    Color Minimum=<rgba 0.8153 0.5361 1 0.15>
    Color Mode=::NewEnumerator2
    Lifetime Max=0.4
    Lifetime Min=0.3
    Lifetime Mode=::NewEnumerator1
    Sprite Size Mode=::NewEnumerator3
    Uniform Sprite Size=&Multiply_Float(A=$System.GlobalScale, B=&RandomRangeFloat(Minimum=100, Maximum=130))
    UsePositionOffset=false
    Write Material Random=true
   )
   &CameraOffset(Camera Offset Amount=&RandomRangeFloat(Minimum=40, Maximum=50, Evaluation Type=::NewEnumerator1))
   &ShapeLocation(
    Hemisphere Distribution=<1 0.5>
    MaskSpawnGroup=false
    Override Seed=false
    Rotation Angle=90
    Rotation Mode=::NewEnumerator1
    Sphere Radius=&MakeCustomFloatFromBool(Bool=$User.bIsFrontEndPreview, True Float=5, False Float=10)
    UseOwnerScale=true
   )
   &AddVelocity(
    Velocity=&Multiply_VectorByFloat(Vector=$Particles.ShapeLocation.ShapeVector, Float=10)
    Velocity Speed Scale=$System.GlobalScale
   )
  ]
  particleUpdateScript [
   &ParticleState()
   &ScaleSpriteSize(Uniform Curve Sprite Scale=DataInterface() {propertyValues {Curve {keys [
        {arriveTangent=1.7761, interpMode=RCIM_Cubic, leaveTangent=1.7761, tangentMode=RCTM_User, time=0, value=0}
        {interpMode=RCIM_Cubic, tangentMode=RCTM_User, time=1, value=1}
       ]}}})
   &ScaleColor(
    Scale Alpha=&FloatFromCurve(
     FloatCurve=DataInterface() {propertyValues {Curve {keys [
         {arriveTangent=2.4358, interpMode=RCIM_Cubic, leaveTangent=2.4358, tangentMode=RCTM_User, time=0, value=0}
         <at 0.25 0.5 RCIM_Cubic>
         <at 1 0 RCIM_Cubic>
        ]}}}
     Scale Curve=&FN_FrontEndFloatModifier(Gameplay=1, Front End=0.7)
    )
    Scale RGB=&VectorFromCurve(VectorCurve=DataInterface() {propertyValues {XCurve {keys [<at 0 1>,<at 1 1>]}, YCurve {keys [<at 0 1>,<at 1 1>]}, ZCurve {keys [<at 0 1>,<at 1 1>]}}})
    ScaleA=true
    ScaleRGB=true
   )
   &DynamicMaterialParameters(
    Index 0 Param 1=&Add_Float(
     A=&RandomRangeFloat(Minimum=0, Maximum=1)
     B=&FloatFromCurve(
      FloatCurve=DataInterface() {propertyValues {Curve {keys [
          <at 0 0 RCIM_Cubic>
          {arriveTangent=1.5947, interpMode=RCIM_Cubic, leaveTangent=1.5947, tangentMode=RCTM_User, time=0.2557, value=0.5257}
          <at 1 1 RCIM_Cubic>
         ]}}}
      CurveIndex=$Particles.Age
      Scale Curve=0.25
     )
    )
    Module.Param0WriteEnabled=true
    Module.Param1WriteEnabled=true
    Module.Param2WriteEnabled=true
    Module.Param3WriteEnabled=true
   ) {module=DynamicMaterialParameters001}
   &SolveForcesAndVelocity(Clamp Velocity=false, Limit Acceleration=false)
  ]
  rendererClasses [&NiagaraSpriteRendererProperties]
  renderers [
   {rendererClass=&NiagaraSpriteRendererProperties}
   {rendererClass=&NiagaraSpriteRendererProperties, rendererIndex=1}
  ]
 }
 Light001 {
  emitterUpdateScript [
   &EmitterState()
   &SpawnBurst_Instantaneous(Spawn Count=1, Use Spawn Probability=false)
   &SpawnBurst_Instantaneous(Spawn Count=2, Spawn Group=1, Use Spawn Probability=false, UseLoopCountLimit=false) {module=SpawnBurst_Instantaneous001}
   &SpawnBurst_Instantaneous(Spawn Count=&RandomRangeInt(Minimum=7, Maximum=9), Spawn Group=2, Use Spawn Probability=false) {module=SpawnBurst_Instantaneous002}
   &SpawnBurst_Instantaneous(Spawn Count=1, Spawn Group=3, Use Spawn Probability=false) {module=SpawnBurst_Instantaneous003}
  ]
  particleSpawnScript [
   /SetVariables(Particles.VisibilityTag=&SpawnGroup()) {isSetParameters=true}
   &InitializeParticle(
    Color=&MaskLinearColorBySpawnGroup(Masked Linear Color=<rgba 0.0464 0 0.1953 2>, Passthrough Linear Color=<rgba 1 1 1 1>, Spawn Group Mask=3)
    Lifetime=&MaskFloatBySpawnGroup(
     Masked Float=0.25
     Passthrough Float=&MaskFloatBySpawnGroup(Masked Float=0.2, Passthrough Float=&RandomRangeFloat(Minimum=0.15, Maximum=0.2), Spawn Group Mask=1)
    )
    Mesh Scale=&Multiply_VectorByFloat(
     Vector=&RandomRangeVector(Minimum=<1.2 1.6 1.6>, Maximum=<0.8 1.6 1.6>)
     Float=&RandomRangeFloat(Minimum=0.8, Maximum=0.9)
    )
    Mesh Scale Mode=::NewEnumerator4
    Sprite Size Max=&MaskVector2dBySpawnGroup(
     Masked Vector2D=<15 65>
     Passthrough Vector2D=&MaskVector2dBySpawnGroup(
      Masked Vector2D=&Vector2DFromFloat(Value=100)
      Passthrough Vector2D=&Vector2DFromFloat(Value=&Multiply_Float(A=50, B=$System.GlobalScale))
      Spawn Group Mask=3
     )
     Spawn Group Mask=2
    )
    Sprite Size Min=&MaskVector2dBySpawnGroup(
     Masked Vector2D=<10 45>
     Passthrough Vector2D=&MaskVector2dBySpawnGroup(
      Masked Vector2D=&Vector2DFromFloat(Value=100)
      Passthrough Vector2D=&Vector2DFromFloat(Value=&Multiply_Float(A=50, B=$System.GlobalScale))
      Spawn Group Mask=3
     )
     Spawn Group Mask=2
    )
    Sprite Size Mode=::NewEnumerator2
    UsePositionOffset=false
    Write Material Random=true
   )
   &ShapeLocation(
    MaskSpawnGroup=false
    Override Seed=false
    Sphere Radius=&MaskFloatBySpawnGroup(Masked Float=3, Passthrough Float=0, Spawn Group Mask=2)
    Sphere Surface Distribution=1
    UseOwnerScale=true
   )
   &AddVelocityFromPoint(
    Module.Use Falloff Distance=false
    Velocity Strength=&MaskFloatBySpawnGroup(Masked Float=&Multiply_Float(A=100, B=$System.GlobalScale), Passthrough Float=0, Spawn Group Mask=2)
   )
   &CameraOffset(Camera Offset Amount=&MaskFloatBySpawnGroup(
     Masked Float=20
     Passthrough Float=&MaskFloatBySpawnGroup(
      Masked Float=&Lerp_Float(A=20, B=4, Alpha=&MakeCustomFloatFromBool(Bool=$User.bIsFrontEndPreview, True Float=1, False Float=0))
      Passthrough Float=&MaskFloatBySpawnGroup(Masked Float=0, Passthrough Float=25, Spawn Group Mask=2)
      Spawn Group Mask=1
     )
    ))
  ]
  particleUpdateScript [
   &ParticleState()
   &ScaleColor(
    Scale Alpha=&MaskFloatBySpawnGroup(
     Masked Float=&FloatFromCurve(
      FloatCurve=DataInterface() {propertyValues {Curve {keys [<at 0 1>,<at 1 0>]}}}
      Scale Curve=&MakeCustomFloatFromBool(Bool=$User.bIsFrontEndPreview, True Float=0.2, False Float=1)
     )
     Passthrough Float=&MaskFloatBySpawnGroup(
      Masked Float=&FloatFromCurve(FloatCurve=DataInterface() {propertyValues {Curve {
          defaultValue=0
          keys [{interpMode=RCIM_Cubic, tangentMode=RCTM_User, time=0.2, value=1}, {interpMode=RCIM_Cubic, tangentMode=RCTM_User, time=1, value=0}]
         }}})
      Passthrough Float=&MaskFloatBySpawnGroup(
       Masked Float=&FloatFromCurve(
        FloatCurve=DataInterface() {propertyValues {Curve {
           defaultValue=1
           keys [
            {arriveTangent=-1.7246, interpMode=RCIM_Cubic, leaveTangent=-1.7246, tangentMode=RCTM_User, time=0.1, value=1}
            {arriveTangentWeight=0.8846, interpMode=RCIM_Cubic, tangentMode=RCTM_User, time=1, value=0}
           ]
          }}}
        Scale Curve=&FN_FrontEndFloatModifier(Gameplay=1.5, Front End=1)
       )
       Passthrough Float=&MaskFloatBySpawnGroup(
        Masked Float=&Multiply_Float(
         A=&FloatFromCurve(
          FloatCurve=DataInterface() {propertyValues {Curve {keys [<at 0 1>,<at 1 0>]}}}
          Scale Curve=&FN_FrontEndFloatModifier(Gameplay=1, Front End=0.2)
         )
         B=1
        )
        Passthrough Float=1
        Spawn Group Mask=3
       )
       Spawn Group Mask=2
      )
      Spawn Group Mask=1
     )
    )
    Scale RGB=&MaskVectorBySpawnGroup(
     Masked Vector=&MakeVectorFromLinearColorRGB(Color=&ColorFromCurve(ColorCurve=DataInterface() {propertyValues {
         AlphaCurve {keys [<at 0 1 RCIM_Cubic>,<at 1 0 RCIM_Cubic>]}
         BlueCurve {keys [
           <at 0 0.8883>
           <at 0.0649 4.101>
           <at 0.125 1>
           {arriveTangent=-1.301, interpMode=RCIM_Cubic, leaveTangent=-1.301, time=0.3915, value=0.5089}
           <at 0.4885 0.5271>
           <at 1 0.5271 RCIM_Cubic>
          ]}
         GreenCurve {keys [
           <at 0 0.8472>
           <at 0.0649 1.8962>
           <at 0.125 0.0436>
           {arriveTangent=-0.0073, interpMode=RCIM_Cubic, leaveTangent=-0.0073, time=0.3915, value=0.0222}
           <at 0.4885 0.0409>
           <at 1 0.0409 RCIM_Cubic>
          ]}
         RedCurve {keys [
           <at 0 0.905>
           <at 0.0649 5>
           <at 0.125 0.9765>
           {arriveTangent=-2.5367, interpMode=RCIM_Cubic, leaveTangent=-2.5367, time=0.3915, value=0.4969}
           <at 0.4885 0.0545>
           <at 1 0.0545 RCIM_Cubic>
          ]}
        }}))
     Passthrough Vector=&MaskVectorBySpawnGroup(
      Masked Vector=&Multiply_VectorByFloat(
       Vector=&MakeVectorFromLinearColorRGB(Color=&ColorFromCurve(ColorCurve=DataInterface() {propertyValues {
           AlphaCurve {keys [<at 0 1 RCIM_Cubic>,<at 1 1 RCIM_Cubic>]}
           BlueCurve {keys [
             <at 0 5>
             {arriveTangent=-5.4763, interpMode=RCIM_Cubic, leaveTangent=-5.4763, time=0.1856, value=5}
             <at 0.4216 1.25>
             <at 0.7304 1>
             <at 0.9431 0.5271 RCIM_Cubic>
            ]}
           GreenCurve {keys [
             <at 0 0.2179>
             {arriveTangent=-0.192, interpMode=RCIM_Cubic, leaveTangent=-0.192, time=0.1856, value=0.2179}
             <at 0.4216 0.0679>
             <at 0.7304 0.0776>
             <at 0.9431 0.0409 RCIM_Cubic>
            ]}
           RedCurve {keys [
             <at 0 4.8826>
             {arriveTangent=-6.5431, interpMode=RCIM_Cubic, leaveTangent=-6.5431, time=0.1856, value=4.8826}
             <at 0.4216 0.8756>
             <at 0.7304 0.1034>
             <at 0.9431 0.0545 RCIM_Cubic>
            ]}
          }}))
       Float=&FN_FrontEndFloatModifier(Gameplay=1, Front End=0.5)
      )
      Passthrough Vector=&MaskVectorBySpawnGroup(
       Masked Vector=&Multiply_VectorByFloat(
        Vector=<1.75 1.75 1.75>
        Float=&FloatFromCurve(
         FloatCurve=DataInterface() {propertyValues {Curve {keys [
             <at 0 1 RCIM_Cubic>
             {arriveTangentWeight=0.1468, interpMode=RCIM_Cubic, leaveTangentWeight=0.0918, tangentMode=RCTM_User, time=0.4, value=0.7}
             {arriveTangentWeight=0.0932, interpMode=RCIM_Cubic, leaveTangentWeight=0.1304, tangentMode=RCTM_User, time=0.65, value=0.7}
             <at 1 1 RCIM_Cubic>
            ]}}}
         CurveIndex=Linked() {linkedVariable {name=NPC.FortniteNPC.TimeOfDay_0-1, type=NiagaraFloat}}
        )
       )
       Passthrough Vector=<1 1 1>
      )
      Spawn Group Mask=1
     )
    )
    ScaleA=true
    ScaleRGB=true
   ) {module=ScaleColor001}
   &ScaleMeshSize(Scale Factor=&VectorFromFloat(Value=&FloatFromCurve(
      FloatCurve=DataInterface() {propertyValues {Curve {
         defaultValue=0.5
         keys [
          {interpMode=RCIM_Cubic, leaveTangentWeight=0.6994, tangentMode=RCTM_User, time=0, value=0.4}
          {arriveTangent=0.9729, interpMode=RCIM_Cubic, leaveTangent=0.9729, tangentMode=RCTM_User, time=1, value=1}
         ]
        }}}
      Scale Curve=$System.GlobalScale
     )))
   &FN_Light_Attributes(
    Color=<rgba 0.3389 0.004 1 1>
    Create Color Attribute=true
    Create Exponent Attribute=true
    Create Position Attribute=true
    Create Radius Attribute=true
    Exponent=5
    Position=<15 0 0>
    Radius=90
    Scale Color=&FloatFromCurve(
     FloatCurve=DataInterface() {propertyValues {Curve {
        defaultValue=1
        keys [{interpMode=RCIM_Cubic, tangentMode=RCTM_User, time=0, value=1}, {interpMode=RCIM_Cubic, tangentMode=RCTM_User, time=1, value=0}]
       }}}
     Scale Curve=10
    )
   )
   &ScaleSpriteSize(Scale Factor=&MaskVector2dBySpawnGroup(
     Masked Vector2D=&Vector2DFromFloat(Value=&FloatFromCurve(
       FloatCurve=DataInterface() {propertyValues {Curve {keys [<at 0 0.6>,<at 1 2>]}}}
       Scale Curve=&Multiply_Float(A=0.75, B=$System.GlobalScale)
      ))
     Passthrough Vector2D=&MaskVector2dBySpawnGroup(
      Masked Vector2D=&Vector2DFromFloat(Value=&FloatFromCurve(FloatCurve=DataInterface() {propertyValues {Curve {keys [<at 0 1>,<at 1 3>]}}}, Scale Curve=$System.GlobalScale))
      Passthrough Vector2D=<1 1>
      Spawn Group Mask=3
     )
     Spawn Group Mask=2
    ))
   &DynamicMaterialParameters(
    Index 0 Param 1=&FloatFromCurve(FloatCurve=DataInterface() {propertyValues {Curve {
        defaultValue=1
        keys [
         {interpMode=RCIM_Cubic, tangentMode=RCTM_User, time=0.4, value=0}
         {arriveTangent=0.8196, interpMode=RCIM_Cubic, leaveTangent=0.8196, tangentMode=RCTM_User, time=1, value=0.5}
        ]
       }}})
    Module.Param0WriteEnabled=true
    Module.Param1WriteEnabled=false
    Module.Param2WriteEnabled=false
    Module.Param3WriteEnabled=false
   )
   &SolveForcesAndVelocity(Clamp Velocity=false, Limit Acceleration=false)
   /SetVariables(Particles.GlowDynamicParameter:Vector4f=&MakeVector4(X=1, Y=1, Z=1, W=1)) {isSetParameters=true}
  ]
  rendererClasses [
   &NiagaraSpriteRendererProperties
   &NiagaraLightRendererProperties
   &NiagaraMeshRendererProperties
  ]
  renderers [
   {rendererClass=&NiagaraSpriteRendererProperties}
   {rendererClass=&NiagaraLightRendererProperties, rendererIndex=1}
   {rendererClass=&NiagaraMeshRendererProperties, rendererIndex=2}
   {rendererClass=&NiagaraSpriteRendererProperties, rendererIndex=3}
   {rendererClass=&NiagaraSpriteRendererProperties, rendererIndex=4}
  ]
 }
 Spark {
  emitterUpdateScript [
   &EmitterState(Life Cycle Mode=::NewEnumerator1, Loop Behavior=::NewEnumerator1, Loop Delay=0.05, Loop Duration=1, UseLoopDelay=true)
   &SpawnBurst_Instantaneous(Spawn Count=25, Use Spawn Probability=false)
  ]
  particleSpawnScript [
   &InitializeParticle(
    Color=&ColorFromCurve(
     ColorCurve=DataInterface() {propertyValues {
       AlphaCurve {keys [<at 0 1 RCIM_Cubic>]}
       BlueCurve {keys [<at 0.0492 0.9773 RCIM_Cubic>,<at 0.2941 1>,<at 0.4897 1>,<at 0.654 1>,<at 0.9139 1 RCIM_Cubic>]}
       GreenCurve {keys [<at 0.0492 0.112 RCIM_Cubic>,<at 0.2941 0.1144>,<at 0.4897 0.1146>,<at 0.654 0.5148>,<at 0.9139 0.8417 RCIM_Cubic>]}
       RedCurve {keys [<at 0.0492 1 RCIM_Cubic>,<at 0.2941 0.4125>,<at 0.4897 0.4105>,<at 0.654 0.1016>,<at 0.9139 0.112 RCIM_Cubic>]}
      }}
     CurveIndex=&RandomRangeFloat(Minimum=0, Maximum=1)
    )
    Lifetime=&FloatFromCurve(
     FloatCurve=DataInterface() {propertyValues {Curve {keys [<at 0.02 1.5>,<at 0.2 0.9>,<at 1 0.5>]}}}
     CurveIndex=&RandomRangeFloat(Minimum=0, Maximum=1)
    )
    Sprite Size Max=<4 13>
    Sprite Size Min=<3 8>
    Sprite Size Mode=::NewEnumerator2
    UsePositionOffset=false
    Write Material Random=true
   )
   &AddVelocityFromPoint(
    Module.Use Falloff Distance=false
    Velocity Strength=&Multiply_Float(A=&RandomRangeFloat(Minimum=400, Maximum=1000), B=$System.GlobalScale)
   )
   /SetVariables() {enabled=false, isSetParameters=true}
  ]
  particleUpdateScript [
   &ParticleState()
   &ScaleColor(
    Scale RGB=&MakeVectorFromLinearColorRGB(Color=&ColorFromCurve(ColorCurve=DataInterface() {propertyValues {
        AlphaCurve {keys [<at 0 1 RCIM_Cubic>]}
        BlueCurve {keys [<at 0 7 RCIM_Cubic>,<at 0.2627 4>,<at 1 0.2839 RCIM_Cubic>]}
        GreenCurve {keys [<at 0 7 RCIM_Cubic>,<at 0.2627 3.9545>,<at 1 0 RCIM_Cubic>]}
        RedCurve {keys [<at 0 7 RCIM_Cubic>,<at 0.2627 3.969>,<at 1 0.0901 RCIM_Cubic>]}
       }}))
    ScaleA=true
    ScaleRGB=true
   )
   &ScaleColor(
    Scale Alpha=&FNTimeOfDayScalar(Night Scale=1, Morning Scale=1, Evening Scale=1)
    Scale RGB=&VectorFromFloat(Value=10)
    ScaleA=true
    ScaleRGB=true
   ) {enabled=false, module=ScaleColor001}
   &Drag(Drag=&RandomRangeFloat(Minimum=12, Maximum=17), Use Linear Drag=true, Use Rotational Drag=false)
   &CurlNoiseForce(
    Noise Strength=&Multiply_Float(
     A=&RandomRangeFloat(Minimum=200, Maximum=400)
     B=&FloatFromCurve(FloatCurve=DataInterface() {propertyValues {Curve {
         defaultValue=0
         keys [
          {interpMode=RCIM_Cubic, tangentMode=RCTM_User, time=0, value=0.5}
          {arriveTangent=5.0731, interpMode=RCIM_Cubic, leaveTangent=5.0731, tangentMode=RCTM_User, time=1, value=2}
         ]
        }}})
    )
    Pan Noise=true
   )
   &GravityForce(Gravity=&RandomRangeVector(Minimum=<0 0 0>, Maximum=<0 0 -100>))
   &SolveForcesAndVelocity(Clamp Velocity=false, Limit Acceleration=false)
   &ScaleSpriteSizeBySpeed(Max Scale Factor=<1 3.5>, Min Scale Factor=<0.5 0.2>)
   &ScaleSpriteSize(Scale Factor=&MakeVector2D(
     X=&FloatFromCurve(
      FloatCurve=DataInterface() {propertyValues {Curve {keys [
          {arriveTangent=9.6837, interpMode=RCIM_Cubic, leaveTangent=9.6837, tangentMode=RCTM_User, time=0, value=0.5}
          <at 0.1 1 RCIM_Cubic>
          <at 1 0.5 RCIM_Cubic>
         ]}}}
      Scale Curve=$System.GlobalScale
     )
     Y=&FloatFromCurve(
      FloatCurve=DataInterface() {propertyValues {Curve {keys [<at 0 1 RCIM_Cubic>, {arriveTangent=-2.9206, interpMode=RCIM_Cubic, leaveTangent=-2.9206, tangentMode=RCTM_User, time=1, value=0}]}}}
      Scale Curve=$System.GlobalScale
     )
    ))
   /SetVariables(Particles.Glow03_Color=&MakeLinearColorFromVectorAndFloat(
     "Vector (RGB)"=&MakeVectorFromLinearColorRGB(Color=&MultiplyLinearColor(
       Linear Color=$Particles.Initial.Color
       Scale Factor=&MakeVector4_FromFloat(Float=&FloatFromCurve(FloatCurve=DataInterface() {propertyValues {Curve {
            defaultValue=1
            keys [{interpMode=RCIM_Cubic, tangentMode=RCTM_User, time=0, value=2}, {interpMode=RCIM_Cubic, tangentMode=RCTM_User, time=1, value=1}]
           }}}))
      ))
     "Float (Alpha)"=&Multiply_Float(
      A=&FN_FrontEndFloatModifier(Gameplay=0.1, Front End=0.03)
      B=&FloatFromCurve(FloatCurve=DataInterface() {propertyValues {Curve {
          defaultValue=1
          keys [
           {interpMode=RCIM_Cubic, tangentMode=RCTM_User, time=-0.0009, value=0.1624}
           {arriveTangent=-0.1622, interpMode=RCIM_Cubic, leaveTangent=-0.1622, time=0.1019, value=0.9636}
           {interpMode=RCIM_Cubic, tangentMode=RCTM_User, time=1, value=0}
          ]
         }}})
     )
    )) {isSetParameters=true}
   /SetVariables(Particles.Glow03_Size=&Multiply_Vector2D_ByFloat(
     Float=&FloatFromCurve(
      FloatCurve=DataInterface() {propertyValues {Curve {keys [
          {arriveTangent=8.6673, interpMode=RCIM_Cubic, leaveTangent=8.6673, tangentMode=RCTM_User, time=0, value=0.5}
          {arriveTangent=-0.5, interpMode=RCIM_Cubic, leaveTangent=-0.5, time=0.2, value=1.3}
          {arriveTangent=-0.1423, interpMode=RCIM_Cubic, leaveTangent=-0.1423, tangentMode=RCTM_User, time=1, value=0}
         ]}}}
      Scale Curve=&Multiply_Float(A=7, B=$System.GlobalScale)
     )
     Vector2D=<1 2>
    )) {isSetParameters=true}
  ]
  rendererClasses [&NiagaraSpriteRendererProperties]
  renderers [
   {rendererClass=&NiagaraSpriteRendererProperties}
   {rendererClass=&NiagaraSpriteRendererProperties, rendererIndex=1}
  ]
 }
 Waves {
  emitterUpdateScript [
   &EmitterState()
   &SpawnBurst_Instantaneous(Use Spawn Probability=false, UseLoopCountLimit=false)
  ]
  particleSpawnScript [&InitializeParticle(
    Color=&Multiply_LinearColorByFloat(
     Linear Color=<rgba 1 1 1 1>
     Float=&FloatFromCurve(
      FloatCurve=DataInterface() {propertyValues {Curve {keys [
          {interpMode=RCIM_Cubic, leaveTangentWeight=0.1, tangentMode=RCTM_User, time=0, value=1.5}
          {arriveTangentWeight=0.1835, interpMode=RCIM_Cubic, leaveTangentWeight=0.2141, tangentMode=RCTM_User, time=0.35, value=1.25}
          {arriveTangentWeight=0.2034, interpMode=RCIM_Cubic, leaveTangentWeight=0.2034, tangentMode=RCTM_User, time=0.65, value=1.25}
          {arriveTangentWeight=0.1167, interpMode=RCIM_Cubic, tangentMode=RCTM_User, time=1, value=1.5}
         ]}}}
      CurveIndex=Linked() {linkedVariable {name=NPC.FortniteNPC.TimeOfDay_0-1, type=NiagaraFloat}}
      Scale Curve=&FN_FrontEndFloatModifier(Gameplay=1, Front End=0.75)
     )
    )
    Lifetime=0.5
    Sprite Size Mode=::NewEnumerator3
    Uniform Sprite Size=&Multiply_Float(A=115, B=$System.GlobalScale)
    UsePositionOffset=false
    Write Material Random=true
   )]
  particleUpdateScript [
   &ParticleState()
   &ScaleSpriteSize(Uniform Curve Sprite Scale=DataInterface() {propertyValues {Curve {keys [
        {interpMode=RCIM_Cubic, leaveTangent=4.8323, tangentMode=RCTM_Break, time=-0.0079, value=0.2118}
        {interpMode=RCIM_Cubic, leaveTangent=0.4119, tangentMode=RCTM_Break, time=0.2, value=0.8}
        {arriveTangent=0.0536, tangentMode=RCTM_User, time=1, value=1}
       ]}}})
   &CameraOffset(Camera Offset Amount=25)
   &ScaleColor(
    Linear Color Curve=DataInterface() {propertyValues {
      AlphaCurve {keys [<at 0 1>,<at 0.5753 0.4>,<at 1 0>]}
      BlueCurve {keys [<at 0 5>,<at 0.3262 0.4271>,<at 1 0.2396>]}
      GreenCurve {keys [<at 0 5>,<at 0.3262 0.4271>,<at 1 0.2396>]}
      RedCurve {keys [<at 0 5>,<at 0.3262 0.4271>,<at 1 0.2396>]}
     }}
    Scale Mode=::NewEnumerator2
   )
   &DynamicMaterialParameters(
    Index 0 Param 1=&FloatFromCurve(FloatCurve=DataInterface() {propertyValues {Curve {keys [
         <at 0.0053 0.1 RCIM_Cubic>
         {arriveTangent=1.5143, interpMode=RCIM_Cubic, leaveTangent=1.5143, time=0.171, value=0.7}
         <at 0.5666 0.95 RCIM_Cubic>
        ]}}})
    Index 0 Param 2=&FloatFromCurve(FloatCurve=DataInterface() {propertyValues {Curve {keys [{interpMode=RCIM_Cubic, leaveTangentWeight=0.5891, tangentMode=RCTM_User, time=0, value=1}, <at 1 0.4 RCIM_Cubic>]}}})
    Index 0 Param 3=&FloatFromCurve(FloatCurve=DataInterface() {propertyValues {Curve {keys [{interpMode=RCIM_Cubic, leaveTangent=-5.0063, tangentMode=RCTM_Break, time=0.1, value=1}, <at 0.6 0>]}}})
    Module.Index0Param0Write=true
    Module.Index0Param1Write=true
    Module.Index0Param2Write=true
    Module.Index0Param3Write=true
   )
  ]
  rendererClasses [&NiagaraSpriteRendererProperties]
  renderers [{rendererClass=&NiagaraSpriteRendererProperties}]
 }
}
summary {
 systemName=NS_Pickaxe_DrumStick_VividSplash_Impact
 emitters [
  {emitterName=Impact_Smoke, rendererClasses [&NiagaraSpriteRendererProperties]}
  {
   emitterName=FN_PhysMatDebris
   rendererClasses [&NiagaraSpriteRendererProperties, &NiagaraMeshRendererProperties]
  }
  {emitterName=Spark, rendererClasses [&NiagaraSpriteRendererProperties]}
  {emitterName=Waves, rendererClasses [&NiagaraSpriteRendererProperties]}
  {
   emitterName=Light001
   rendererClasses [
    &NiagaraSpriteRendererProperties
    &NiagaraLightRendererProperties
    &NiagaraMeshRendererProperties
   ]
  }
 ]
 userVariables [
  User.bIsFrontEnd {type=NiagaraBool}
  User.bIsFrontEndPreview {type=NiagaraBool}
  User.FrontEndScale {type=NiagaraFloat, defaultValue=0.6, description=""}
  User.PhysicalSurface {type=NiagaraInt32}
 ]
}
